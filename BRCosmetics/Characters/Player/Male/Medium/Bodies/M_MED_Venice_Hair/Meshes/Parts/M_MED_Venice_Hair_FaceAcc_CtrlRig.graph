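# Control Rig for the M_MED_Venice_Hair face accessory that rotates the six back hair master bones according to head orientation.
# A spherical pose reader on the head outputs a weight that blends a fixed X-axis rotation onto each dyn_hair_B master bone.
# When IsSkyDivingCtrl is true, a second reader with a different rotation offset then applies a smaller rotation to the same bones.
# The front hair and feather bones are left untouched.
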ControlRigBlueprint /BRCosmetics/Characters/Player/Male/Medium/Bodies/M_MED_Venice_Hair/Meshes/Parts/M_MED_Venice_Hair_FaceAcc_CtrlRig
graphs [
 RigVMFunctionLibrary {links [], nodes []}
 RigVMModel {nodes [
   RigUnit_BeginExecution = &RigUnit_BeginExecution()
   SequenceAggregate = &RigVMFunction_Sequence(ExecuteContext=@RigUnit_BeginExecution.ExecutePin)
   n1 = &RigUnit_SphericalPoseReader(
    ActiveRegionScaleFactors=(PositiveWidth=0.5,NegativeWidth=1,PositiveHeight=1,NegativeHeight=1)
    ActiveRegionSize=0.1
    Debug=(bDrawDebug=false,DebugScale=25,DebugSegments=20,DebugThickness=0.25)
    DriverAxis=(X=0,Y=-1,Z=0)
    DriverItem=(Type=Bone,Name="head")
    FalloffRegionScaleFactors=(PositiveWidth=1,NegativeWidth=1,PositiveHeight=1,NegativeHeight=1)
    FalloffSize=0.2
    OptionalParentItem=(Type=Bone)
    RotationOffset=(X=-90,Y=0,Z=-60)
    ExecutePin=@SequenceAggregate.A
   )
   n2 = &RigVMFunction_MathQuaternionFromEuler(Euler=(X=-90,Y=0,Z=0))
   > &RigUnit_ModifyTransforms(
    ItemToModify=((Item=(Type=Bone,Name="dyn_hair_B_01_mstr"),Transform=(Rotation=(X=0,Y=0,Z=0,W=1),Translation=(X=0,Y=0,Z=0),Scale3D=(X=1,Y=1,Z=1))))
    ItemToModify.0.Transform.Rotation=@n2.Result
    Weight=@n1.OutputParam
   )
   > &RigUnit_ModifyTransforms(
    ItemToModify=((Item=(Type=Bone,Name="dyn_hair_B_02_mstr"),Transform=(Rotation=(X=0,Y=0,Z=0,W=1),Translation=(X=0,Y=0,Z=0),Scale3D=(X=1,Y=1,Z=1))))
    ItemToModify.0.Transform.Rotation=@n2.Result
    Weight=@n1.OutputParam
   )
   > &RigUnit_ModifyTransforms(
    ItemToModify=((Item=(Type=Bone,Name="dyn_hair_B_03_mstr"),Transform=(Rotation=(X=0,Y=0,Z=0,W=1),Translation=(X=0,Y=0,Z=0),Scale3D=(X=1,Y=1,Z=1))))
    ItemToModify.0.Transform.Rotation=@n2.Result
    Weight=@n1.OutputParam
   )
   > &RigUnit_ModifyTransforms(
    ItemToModify=((Item=(Type=Bone,Name="dyn_hair_B_04_mstr"),Transform=(Rotation=(X=0,Y=0,Z=0,W=1),Translation=(X=0,Y=0,Z=0),Scale3D=(X=1,Y=1,Z=1))))
    ItemToModify.0.Transform.Rotation=@n2.Result
    Weight=@n1.OutputParam
   )
   > &RigUnit_ModifyTransforms(
    ItemToModify=((Item=(Type=Bone,Name="dyn_hair_B_05_mstr"),Transform=(Rotation=(X=0,Y=0,Z=0,W=1),Translation=(X=0,Y=0,Z=0),Scale3D=(X=1,Y=1,Z=1))))
    ItemToModify.0.Transform.Rotation=@n2.Result
    Weight=@n1.OutputParam
   )
   > &RigUnit_ModifyTransforms(
    ItemToModify=((Item=(Type=Bone,Name="dyn_hair_B_06_mstr"),Transform=(Rotation=(X=0,Y=0,Z=0,W=1),Translation=(X=0,Y=0,Z=0),Scale3D=(X=1,Y=1,Z=1))))
    ItemToModify.0.Transform.Rotation=@n2.Result
    Weight=@n1.OutputParam
   )
   Branch_RigVMUnitNode = &RigVMFunction_ControlFlowBranch(ExecuteContext=@SequenceAggregate.B, Condition=RigVMVariableNode() {variable=IsSkyDivingCtrl}.Value)
   n3 = &RigUnit_SphericalPoseReader(
    ActiveRegionScaleFactors=(PositiveWidth=0.5,NegativeWidth=1,PositiveHeight=1,NegativeHeight=1)
    ActiveRegionSize=0.1
    Debug=(bDrawDebug=false,DebugScale=25,DebugSegments=20,DebugThickness=0.25)
    DriverAxis=(X=0,Y=-1,Z=0)
    DriverItem=(Type=Bone,Name="head")
    FalloffRegionScaleFactors=(PositiveWidth=1,NegativeWidth=1,PositiveHeight=1,NegativeHeight=1)
    FalloffSize=0.2
    OptionalParentItem=(Type=Bone)
    RotationOffset=(X=-90,Y=0,Z=-30)
    ExecutePin=@Branch_RigVMUnitNode.True
   )
   n4 = &RigVMFunction_MathQuaternionFromEuler(Euler=(X=-30,Y=0,Z=0))
   > &RigUnit_ModifyTransforms(
    ItemToModify=((Item=(Type=Bone,Name="dyn_hair_B_01_mstr"),Transform=(Rotation=(X=0,Y=0,Z=0,W=1),Translation=(X=0,Y=0,Z=0),Scale3D=(X=1,Y=1,Z=1))))
    ItemToModify.0.Transform.Rotation=@n4.Result
    Weight=@n3.OutputParam
   )
   > &RigUnit_ModifyTransforms(
    ItemToModify=((Item=(Type=Bone,Name="dyn_hair_B_02_mstr"),Transform=(Rotation=(X=0,Y=0,Z=0,W=1),Translation=(X=0,Y=0,Z=0),Scale3D=(X=1,Y=1,Z=1))))
    ItemToModify.0.Transform.Rotation=@n4.Result
    Weight=@n3.OutputParam
   )
   > &RigUnit_ModifyTransforms(
    ItemToModify=((Item=(Type=Bone,Name="dyn_hair_B_03_mstr"),Transform=(Rotation=(X=0,Y=0,Z=0,W=1),Translation=(X=0,Y=0,Z=0),Scale3D=(X=1,Y=1,Z=1))))
    ItemToModify.0.Transform.Rotation=@n4.Result
    Weight=@n3.OutputParam
   )
   > &RigUnit_ModifyTransforms(
    ItemToModify=((Item=(Type=Bone,Name="dyn_hair_B_04_mstr"),Transform=(Rotation=(X=0,Y=0,Z=0,W=1),Translation=(X=0,Y=0,Z=0),Scale3D=(X=1,Y=1,Z=1))))
    ItemToModify.0.Transform.Rotation=@n4.Result
    Weight=@n3.OutputParam
   )
   > &RigUnit_ModifyTransforms(
    ItemToModify=((Item=(Type=Bone,Name="dyn_hair_B_05_mstr"),Transform=(Rotation=(X=0,Y=0,Z=0,W=1),Translation=(X=0,Y=0,Z=0),Scale3D=(X=1,Y=1,Z=1))))
    ItemToModify.0.Transform.Rotation=@n4.Result
    Weight=@n3.OutputParam
   )
   > &RigUnit_ModifyTransforms(
    ItemToModify=((Item=(Type=Bone,Name="dyn_hair_B_06_mstr"),Transform=(Rotation=(X=0,Y=0,Z=0,W=1),Translation=(X=0,Y=0,Z=0),Scale3D=(X=1,Y=1,Z=1))))
    ItemToModify.0.Transform.Rotation=@n4.Result
    Weight=@n3.OutputParam
   )
  ]}
]
hierarchy [
 root {loc=<0 0 0>, rot=<0 0 0 1>}
 pelvis {parent=root, loc=<0 -0.7071 93.1191>, rot=<0.0127 0.707 -0.0127 -0.707>}
 spine_01 {parent=pelvis, loc=<6.7494 -0.6388 0>, rot=<0 0 0.034 -0.9994>}
 spine_02 {parent=spine_01, loc=<7.2518 -0.035 0>, rot=<0 0 0.0238 -0.9997>}
 spine_03 {parent=spine_02, loc=<6.0104 0 0>, rot=<0 0 0.0149 -0.9999>}
 spine_04 {parent=spine_03, loc=<8.1024 0 0>, rot=<0 0 -0.0132 -0.9999>}
 spine_05 {parent=spine_04, loc=<10.0146 0 0>, rot=<0 0 -0.0015 -1>}
 neck_01 {parent=spine_05, loc=<11.3138 0.1529 0>, rot=<0 0 -0.2003 -0.9797>}
 neck_02 {parent=neck_01, loc=<4.5627 0 0>, rot=<0 0 0.0504 -0.9987>}
 head {parent=neck_02, loc=<5.3914 0 0>, rot=<0 0 0.115 -0.9934>}
 dyn_hair_FL_mstr {parent=head, loc=<6.0636 -6.7807 -5.043>, rot=<0.0029 -0.7071 -0.0029 -0.7071>}
 dyn_hair_FL_01 {parent=dyn_hair_FL_mstr, loc=<0 0 0>, rot=<-0.1531 -0.5957 -0.1963 -0.7637>}
 dyn_hair_FL_02 {parent=dyn_hair_FL_01, loc=<6.8132 0 0>, rot=<-0.0293 -0.0507 -0.0205 -0.9981>}
 dyn_hair_FL_03 {parent=dyn_hair_FL_02, loc=<8.3061 0 0>, rot=<-0.0249 -0.0404 -0.0076 -0.9988>}
 dyn_hair_FL_04 {parent=dyn_hair_FL_03, loc=<7.9646 0 0>, rot=<-0.0092 -0.0142 -0.0159 -0.9997>}
 dyn_hair_FL_05 {parent=dyn_hair_FL_04, loc=<8.8138 -0.2262 0>, rot=<0.014 0.0279 0.1243 -0.9918>}
 dyn_hair_FL_06 {parent=dyn_hair_FL_05, loc=<8.5691 0.5535 0.0116>, rot=<0 0 0 1>}
 dyn_hair_B_01_mstr {parent=head, loc=<5.9277 -7.5599 -3.0689>, rot=<0.0029 -0.7071 -0.0029 -0.7071>}
 dyn_hair_B_01_01 {parent=dyn_hair_B_01_mstr, loc=<0 0 0>, rot=<-0.393 -0.4376 0.4391 -0.6791>}
 dyn_hair_B_01_02 {parent=dyn_hair_B_01_01, loc=<8.091 0 0>, rot=<-0.0016 0.0266 -0.0103 -0.9996>}
 dyn_hair_B_01_03 {parent=dyn_hair_B_01_02, loc=<7.5968 0 0>, rot=<-0.009 -0.0252 0.0188 -0.9995>}
 dyn_hair_B_01_04 {parent=dyn_hair_B_01_03, loc=<5.8475 0 0>, rot=<-0.0166 0 0 -0.9999>}
 dyn_hair_B_02_mstr {parent=head, loc=<5.7271 -8.7458 -1.7174>, rot=<0.0029 -0.7071 -0.0029 -0.7071>}
 dyn_hair_B_02_01 {parent=dyn_hair_B_02_mstr, loc=<0 0 0>, rot=<-0.3544 -0.4715 0.4319 -0.6823>}
 dyn_hair_B_02_02 {parent=dyn_hair_B_02_01, loc=<8.0056 0 0>, rot=<0.0084 0.0078 0.0261 -0.9996>}
 dyn_hair_B_02_03 {parent=dyn_hair_B_02_02, loc=<7.3666 0 0>, rot=<0.0098 -0.0128 -0.0035 -0.9999>}
 dyn_hair_B_02_04 {parent=dyn_hair_B_02_03, loc=<5.7627 0 0>, rot=<-0.0099 0 0 -1>}
 dyn_hair_B_03_mstr {parent=head, loc=<5.7216 -8.0893 -0.9195>, rot=<0.0029 -0.7071 -0.0029 -0.7071>}
 dyn_hair_B_03_01 {parent=dyn_hair_B_03_mstr, loc=<0 0 0>, rot=<-0.4285 -0.4208 0.5443 -0.5857>}
 dyn_hair_B_03_02 {parent=dyn_hair_B_03_01, loc=<7.9234 0 0>, rot=<0.0107 0.0076 -0.0061 -0.9999>}
 dyn_hair_B_03_03 {parent=dyn_hair_B_03_02, loc=<7.3107 0 0>, rot=<-0.0127 -0.0087 0.017 -0.9997>}
 dyn_hair_B_03_04 {parent=dyn_hair_B_03_03, loc=<5.9889 0 0>, rot=<0.0026 0 0 -1>}
 dyn_hair_B_04_mstr {parent=head, loc=<5.7273 -8.7712 0.4642>, rot=<0.0029 -0.7071 -0.0029 -0.7071>}
 dyn_hair_B_04_01 {parent=dyn_hair_B_04_mstr, loc=<0 0 0>, rot=<-0.4011 -0.419 0.598 -0.5531>}
 dyn_hair_B_04_02 {parent=dyn_hair_B_04_01, loc=<9.1108 0 0>, rot=<0.0138 0.0124 -0.0163 -0.9997>}
 dyn_hair_B_04_03 {parent=dyn_hair_B_04_02, loc=<8.788 0 0>, rot=<-0.0018 0.0043 -0.0047 -1>}
 dyn_hair_B_04_04 {parent=dyn_hair_B_04_03, loc=<7.3576 0 0>, rot=<0 0 0 1>}
 dyn_hair_B_05_mstr {parent=head, loc=<5.5258 -7.9001 1.5187>, rot=<0.0029 -0.7071 -0.0029 -0.7071>}
 dyn_hair_B_05_01 {parent=dyn_hair_B_05_mstr, loc=<0 0 0>, rot=<-0.3712 -0.467 0.6018 -0.5309>}
 dyn_hair_B_05_02 {parent=dyn_hair_B_05_01, loc=<8.2654 0 0>, rot=<-0.0008 0.0056 -0.0027 -1>}
 dyn_hair_B_05_03 {parent=dyn_hair_B_05_02, loc=<7.9491 0 0>, rot=<-0.0008 -0.0309 -0.003 -0.9995>}
 dyn_hair_B_05_04 {parent=dyn_hair_B_05_03, loc=<5.5551 0 0>, rot=<0 0 0 1>}
 dyn_hair_B_06_mstr {parent=head, loc=<5.531 -8.5354 2.5976>, rot=<0.0029 -0.7071 -0.0029 -0.7071>}
 dyn_hair_B_06_01 {parent=dyn_hair_B_06_mstr, loc=<0 0 0>, rot=<-0.5078 -0.251 0.7706 -0.2921>}
 dyn_hair_B_06_02 {parent=dyn_hair_B_06_01, loc=<7.2102 0 0>, rot=<-0.0008 0.0239 -0.0292 -0.9993>}
 dyn_hair_B_06_03 {parent=dyn_hair_B_06_02, loc=<6.9209 0 0>, rot=<-0.0053 -0.0225 -0.0057 -0.9997>}
 dyn_hair_B_06_04 {parent=dyn_hair_B_06_03, loc=<5.9728 0 0>, rot=<0 0 0 1>}
 dyn_hair_FR_mstr {parent=head, loc=<6.064 -6.7807 5.043>, rot=<0.0029 -0.7071 -0.0029 -0.7071>}
 dyn_hair_FR_01 {parent=dyn_hair_FR_mstr, loc=<0 0 0>, rot=<-0.1963 -0.7637 -0.1531 -0.5957>}
 dyn_hair_FR_02 {parent=dyn_hair_FR_01, loc=<6.8132 0 0>, rot=<0.0293 0.0507 -0.0205 -0.9981>}
 dyn_hair_FR_03 {parent=dyn_hair_FR_02, loc=<8.3067 -0.0004 0.0001>, rot=<0.0249 0.0404 -0.0076 -0.9988>}
 dyn_hair_FR_04 {parent=dyn_hair_FR_03, loc=<7.9643 0.0002 0>, rot=<0.0092 0.0142 -0.0159 -0.9997>}
 dyn_hair_FR_05 {parent=dyn_hair_FR_04, loc=<8.8042 -0.3449 0>, rot=<-0.014 -0.0279 0.1243 -0.9918>}
 dyn_hair_FR_06 {parent=dyn_hair_FR_05, loc=<6.0344 0.6349 -0.1625>, rot=<0 0 0 1>}
 dyn_feather_mstr {parent=head, loc=<11.4059 -1.443 -6.3212>, rot=<0.0029 -0.7071 -0.0029 -0.7071>}
 dyn_feather_01 {parent=dyn_feather_mstr, loc=<0 0 0>, rot=<0.1985 -0.5854 0.2524 -0.7444>}
 dyn_feather_02 {parent=dyn_feather_01, loc=<8.1773 0 0>, rot=<0 0 0 1>}
 clavicle_l {parent=spine_05, loc=<4.3575 -0.1466 -2.3223>, rot=<0.0465 -0.7289 -0.009 -0.6829>}
 upperarm_l {parent=clavicle_l, loc=<15.5111 0 0>, rot=<0.0226 -0.3234 -0.0088 -0.9459>}
 clavicle_r {parent=spine_05, loc=<4.3575 -0.1466 2.3223>, rot=<0.7289 0.0465 -0.6829 0.009>}
 upperarm_r {parent=clavicle_r, loc=<-15.5111 0 0>, rot=<0.0226 -0.3234 -0.0088 -0.9459>}
]
variables [IsSkyDivingCtrl {type=bool, default=False}]
