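# Physics asset for the Venice Hair male medium cosmetic's FaceAcc mesh, defining capsule bodies and constraints for its dynamic hair bones.
# Six back chains and longer front-left and front-right chains each hang from a master bone linked to the head with free angular motion.
# Links below each master have tight twist limits and swing limits that generally widen toward the tips.
# Head, upper spine and upper-arm capsules supply body shapes around the hair.
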PhysicsAsset /BRCosmetics/Characters/Player/Male/Medium/Bodies/M_MED_Venice_Hair/Meshes/Parts/M_MED_Venice_Hair_FaceAcc_FE_Physics
bodies {
 dyn_hair_B_01_01 [{shapeType=Capsule, length=4, radius=2, shapeName=dyn_hair_B_01_01_capsule, center=<2 0 0>, rotation=<rot 90 0 0>}]
 dyn_hair_B_01_02 [{shapeType=Capsule, length=4, radius=2, shapeName=dyn_hair_B_01_02_capsule, center=<2 0 0>, rotation=<rot 90 0 0>}]
 dyn_hair_B_01_03 [{shapeType=Capsule, length=4, radius=2, shapeName=dyn_hair_B_01_03_capsule, center=<2 0 0>, rotation=<rot 90 0 0>}]
 dyn_hair_B_01_04 [{shapeType=Capsule, length=4, radius=2, shapeName=dyn_hair_B_01_04_capsule, center=<2 0 0>, rotation=<rot 90 0 0>}]
 dyn_hair_B_01_mstr [{shapeType=Capsule, length=4, radius=2, shapeName=dyn_hair_B_01_mstr_capsule, center=<2 0 0>, rotation=<rot 90 0 0>}]
 dyn_hair_B_02_01 [{shapeType=Capsule, length=4, radius=2, shapeName=dyn_hair_B_02_01_capsule, center=<2 0 0>, rotation=<rot 90 0 0>}]
 dyn_hair_B_02_02 [{shapeType=Capsule, length=4, radius=2, shapeName=dyn_hair_B_02_02_capsule, center=<2 0 0>, rotation=<rot 90 0 0>}]
 dyn_hair_B_02_03 [{shapeType=Capsule, length=4, radius=2, shapeName=dyn_hair_B_02_03_capsule, center=<2 0 0>, rotation=<rot 90 0 0>}]
 dyn_hair_B_02_04 [{shapeType=Capsule, length=4, radius=2, shapeName=dyn_hair_B_02_04_capsule, center=<2 0 0>, rotation=<rot 90 0 0>}]
 dyn_hair_B_02_mstr [{shapeType=Capsule, length=4, radius=2, shapeName=dyn_hair_B_02_mstr_capsule, center=<2 0 0>, rotation=<rot 90 0 0>}]
 dyn_hair_B_03_01 [{shapeType=Capsule, length=4, radius=2, shapeName=dyn_hair_B_03_01_capsule, center=<2 0 0>, rotation=<rot 90 0 0>}]
 dyn_hair_B_03_02 [{shapeType=Capsule, length=4, radius=2, shapeName=dyn_hair_B_03_02_capsule, center=<2 0 0>, rotation=<rot 90 0 0>}]
 dyn_hair_B_03_03 [{shapeType=Capsule, length=4, radius=2, shapeName=dyn_hair_B_03_03_capsule, center=<2 0 0>, rotation=<rot 90 0 0>}]
 dyn_hair_B_03_04 [{shapeType=Capsule, length=4, radius=2, shapeName=dyn_hair_B_03_04_capsule, center=<2 0 0>, rotation=<rot 90 0 0>}]
 dyn_hair_B_03_mstr [{shapeType=Capsule, length=4, radius=2, shapeName=dyn_hair_B_03_mstr_capsule, center=<2 0 0>, rotation=<rot 90 0 0>}]
 dyn_hair_B_04_01 [{shapeType=Capsule, length=4, radius=2, shapeName=dyn_hair_B_04_01_capsule, center=<2 0 0>, rotation=<rot 90 0 0>}]
 dyn_hair_B_04_02 [{shapeType=Capsule, length=4, radius=2, shapeName=dyn_hair_B_04_02_capsule, center=<2 0 0>, rotation=<rot 90 0 0>}]
 dyn_hair_B_04_03 [{shapeType=Capsule, length=4, radius=2, shapeName=dyn_hair_B_04_03_capsule, center=<2 0 0>, rotation=<rot 90 0 0>}]
 dyn_hair_B_04_04 [{shapeType=Capsule, length=4, radius=2, shapeName=dyn_hair_B_04_04_capsule, center=<2 0 0>, rotation=<rot 90 0 0>}]
 dyn_hair_B_04_mstr [{shapeType=Capsule, length=4, radius=2, shapeName=dyn_hair_B_04_mstr_capsule, center=<2 0 0>, rotation=<rot 90 0 0>}]
 dyn_hair_B_05_01 [{shapeType=Capsule, length=4, radius=2, shapeName=dyn_hair_B_05_01_capsule, center=<2 0 0>, rotation=<rot 90 0 0>}]
 dyn_hair_B_05_02 [{shapeType=Capsule, length=4, radius=2, shapeName=dyn_hair_B_05_02_capsule, center=<2 0 0>, rotation=<rot 90 0 0>}]
 dyn_hair_B_05_03 [{shapeType=Capsule, length=4, radius=2, shapeName=dyn_hair_B_05_03_capsule, center=<2 0 0>, rotation=<rot 90 0 0>}]
 dyn_hair_B_05_04 [{shapeType=Capsule, length=4, radius=2, shapeName=dyn_hair_B_05_04_capsule, center=<2 0 0>, rotation=<rot 90 0 0>}]
 dyn_hair_B_05_mstr [{shapeType=Capsule, length=4, radius=2, shapeName=dyn_hair_B_05_mstr_capsule, center=<2 0 0>, rotation=<rot 90 0 0>}]
 dyn_hair_B_06_01 [{shapeType=Capsule, length=4, radius=2, shapeName=dyn_hair_B_06_01_capsule, center=<2 0 0>, rotation=<rot 90 0 0>}]
 dyn_hair_B_06_02 [{shapeType=Capsule, length=4, radius=2, shapeName=dyn_hair_B_06_02_capsule, center=<2 0 0>, rotation=<rot 90 0 0>}]
 dyn_hair_B_06_03 [{shapeType=Capsule, length=4, radius=2, shapeName=dyn_hair_B_06_03_capsule, center=<2 0 0>, rotation=<rot 90 0 0>}]
 dyn_hair_B_06_04 [{shapeType=Capsule, length=4, radius=2, shapeName=dyn_hair_B_06_04_capsule, center=<2 0 0>, rotation=<rot 90 0 0>}]
 dyn_hair_B_06_mstr [{shapeType=Capsule, length=4, radius=2, shapeName=dyn_hair_B_06_mstr_capsule, center=<2 0 0>, rotation=<rot 90 0 0>}]
 dyn_hair_FL_01 [{shapeType=Capsule, length=4, radius=2, shapeName=dyn_hair_FL_01_capsule, center=<2 0 0>, rotation=<rot 90 0 0>}]
 dyn_hair_FL_02 [{shapeType=Capsule, length=6, radius=1.7, shapeName=dyn_hair_FL_02_capsule, center=<2 0 0>, rotation=<rot 90 0 0>}]
 dyn_hair_FL_03 [{shapeType=Capsule, length=6, radius=1.7, shapeName=dyn_hair_FL_03_capsule, center=<2 0 0>, rotation=<rot 90 0 0>}]
 dyn_hair_FL_04 [{shapeType=Capsule, length=3.75, radius=2.75, shapeName=dyn_hair_FL_04_capsule, center=<2 0 0>, rotation=<rot 90 0 0>}]
 dyn_hair_FL_05 [{shapeType=Capsule, length=3.75, radius=3.25, shapeName=dyn_hair_FL_05_capsule, center=<2 0 0>, rotation=<rot 90 0 0>}]
 dyn_hair_FL_06 [{shapeType=Capsule, length=2.5, radius=4, shapeName=dyn_hair_FL_06_capsule, center=<2 0 0>, rotation=<rot 90 0 0>}]
 dyn_hair_FL_mstr [{shapeType=Capsule, length=4, radius=2, shapeName=dyn_hair_FL_mstr_capsule, center=<2 0 0>, rotation=<rot 90 0 0>}]
 dyn_hair_FR_01 [{shapeType=Capsule, length=4, radius=2, shapeName=dyn_hair_FL_01_capsule, center=<2 0 0>, rotation=<rot 90 0 0>}]
 dyn_hair_FR_02 [{shapeType=Capsule, length=4, radius=2, shapeName=dyn_hair_FL_02_capsule, center=<2 0 0>, rotation=<rot 90 0 0>}]
 dyn_hair_FR_03 [{shapeType=Capsule, length=4, radius=2.3, shapeName=dyn_hair_FL_03_capsule, center=<2 0 0>, rotation=<rot 90 0 0>}]
 dyn_hair_FR_04 [{shapeType=Capsule, length=2.75, radius=3.25, shapeName=dyn_hair_FL_04_capsule, center=<2 0 0>, rotation=<rot 90 0 0>}]
 dyn_hair_FR_05 [{shapeType=Capsule, length=0, radius=3.5, shapeName=dyn_hair_FL_05_capsule, center=<1.2536 0 0>, rotation=<rot 90 0 0>}]
 dyn_hair_FR_06 [{shapeType=Capsule, length=2.75, radius=3.75, shapeName=dyn_hair_FL_06_capsule, center=<2 0 0>, rotation=<rot 90 0 0>}]
 dyn_hair_FR_mstr [{shapeType=Capsule, length=4, radius=2, shapeName=dyn_hair_FL_mstr_capsule, center=<2 0 0>, rotation=<rot 90 0 0>}]
 head [{shapeType=Capsule, length=8.5, radius=5.1597, rotation=<rot 90 0 -15.2774>}]
 spine_04 [{shapeType=Capsule, length=37.255, radius=11.75, shapeName=spine_04_capsule, center=<1.1386 -2.9064 0>, rotation=<rot 0 -90 0>}]
 spine_05 [
  {shapeType=Capsule, length=44, radius=9, shapeName=spine_05_capsule, center=<1.4288 -1.9971 0>, rotation=<rot 0 -90 0>}
  {shapeType=Capsule, length=36.25, radius=4.5, shapeName=spine_05_capsule, center=<10.0155 -2.7649 0>}
 ]
 upperarm_l [{shapeType=Capsule, length=14, radius=6, shapeName=upperarm_l_capsule, center=<6.5 -0.792 0>, rotation=<rot 90 0 0>}]
 upperarm_r [{shapeType=Capsule, length=12.75, radius=6.5, shapeName=upperarm_r_capsule, center=<-6 0 0>, rotation=<rot 90 0 0>}]
}
constraints [
 {
  bone1Name=dyn_hair_B_01_mstr
  bone2Name=head
  swing1LimitDegrees=45
  swing1Motion=Free
  swing2LimitDegrees=45
  swing2Motion=Free
  twistLimitDegrees=45
  twistMotion=Free
 }
 {bone1Name=dyn_hair_B_01_03, bone2Name=dyn_hair_B_01_02, swing1LimitDegrees=25, swing2LimitDegrees=25, twistLimitDegrees=2}
 {bone1Name=dyn_hair_B_01_04, bone2Name=dyn_hair_B_01_03, swing1LimitDegrees=30, swing2LimitDegrees=30, twistLimitDegrees=2}
 {
  bone1Name=dyn_hair_B_02_mstr
  bone2Name=head
  swing1LimitDegrees=45
  swing1Motion=Free
  swing2LimitDegrees=45
  swing2Motion=Free
  twistLimitDegrees=45
  twistMotion=Free
 }
 {bone1Name=dyn_hair_B_02_03, bone2Name=dyn_hair_B_02_02, swing1LimitDegrees=25, swing2LimitDegrees=25, twistLimitDegrees=2}
 {bone1Name=dyn_hair_B_02_04, bone2Name=dyn_hair_B_02_03, swing1LimitDegrees=30, swing2LimitDegrees=30, twistLimitDegrees=2}
 {
  bone1Name=dyn_hair_B_03_mstr
  bone2Name=head
  swing1LimitDegrees=45
  swing1Motion=Free
  swing2LimitDegrees=45
  swing2Motion=Free
  twistLimitDegrees=45
  twistMotion=Free
 }
 {bone1Name=dyn_hair_B_03_03, bone2Name=dyn_hair_B_03_02, swing1LimitDegrees=25, swing2LimitDegrees=25, twistLimitDegrees=2}
 {bone1Name=dyn_hair_B_03_04, bone2Name=dyn_hair_B_03_03, swing1LimitDegrees=30, swing2LimitDegrees=30, twistLimitDegrees=2}
 {
  bone1Name=dyn_hair_B_04_mstr
  bone2Name=head
  swing1LimitDegrees=45
  swing1Motion=Free
  swing2LimitDegrees=45
  swing2Motion=Free
  twistLimitDegrees=45
  twistMotion=Free
 }
 {bone1Name=dyn_hair_B_04_03, bone2Name=dyn_hair_B_04_02, swing1LimitDegrees=25, swing2LimitDegrees=25, twistLimitDegrees=2}
 {bone1Name=dyn_hair_B_04_04, bone2Name=dyn_hair_B_04_03, swing1LimitDegrees=30, swing2LimitDegrees=30, twistLimitDegrees=2}
 {
  bone1Name=dyn_hair_B_05_mstr
  bone2Name=head
  swing1LimitDegrees=45
  swing1Motion=Free
  swing2LimitDegrees=45
  swing2Motion=Free
  twistLimitDegrees=45
  twistMotion=Free
 }
 {bone1Name=dyn_hair_B_05_03, bone2Name=dyn_hair_B_05_02, swing1LimitDegrees=25, swing2LimitDegrees=25, twistLimitDegrees=2}
 {bone1Name=dyn_hair_B_05_04, bone2Name=dyn_hair_B_05_03, swing1LimitDegrees=30, swing2LimitDegrees=30, twistLimitDegrees=2}
 {
  bone1Name=dyn_hair_B_06_mstr
  bone2Name=head
  swing1LimitDegrees=45
  swing1Motion=Free
  swing2LimitDegrees=45
  swing2Motion=Free
  twistLimitDegrees=45
  twistMotion=Free
 }
 {bone1Name=dyn_hair_B_06_03, bone2Name=dyn_hair_B_06_02, swing1LimitDegrees=25, swing2LimitDegrees=25, twistLimitDegrees=2}
 {bone1Name=dyn_hair_B_06_04, bone2Name=dyn_hair_B_06_03, swing1LimitDegrees=30, swing2LimitDegrees=30, twistLimitDegrees=2}
 {
  bone1Name=dyn_hair_FL_mstr
  bone2Name=head
  swing1LimitDegrees=45
  swing1Motion=Free
  swing2LimitDegrees=45
  swing2Motion=Free
  twistLimitDegrees=45
  twistMotion=Free
 }
 {bone1Name=dyn_hair_FL_01, bone2Name=dyn_hair_FL_mstr, swing1LimitDegrees=10, swing2LimitDegrees=10, twistLimitDegrees=2}
 {bone1Name=dyn_hair_FL_02, bone2Name=dyn_hair_FL_01, swing1LimitDegrees=20, swing2LimitDegrees=20, twistLimitDegrees=2}
 {bone1Name=dyn_hair_FL_03, bone2Name=dyn_hair_FL_02, swing1LimitDegrees=20, swing2LimitDegrees=20, twistLimitDegrees=2}
 {bone1Name=dyn_hair_FL_04, bone2Name=dyn_hair_FL_03, swing1LimitDegrees=30, swing2LimitDegrees=30, twistLimitDegrees=2}
 {bone1Name=dyn_hair_FL_05, bone2Name=dyn_hair_FL_04, swing1LimitDegrees=30, swing2LimitDegrees=30, twistLimitDegrees=2}
 {bone1Name=dyn_hair_FL_06, bone2Name=dyn_hair_FL_05, swing1LimitDegrees=45, swing2LimitDegrees=45, twistLimitDegrees=2}
 {
  bone1Name=head
  bone2Name=spine_05
  swing1LimitDegrees=45
  swing1Motion=Free
  swing2LimitDegrees=45
  swing2Motion=Free
  twistLimitDegrees=45
  twistMotion=Free
 }
 {
  bone1Name=spine_05
  bone2Name=spine_04
  swing1LimitDegrees=45
  swing1Motion=Free
  swing2LimitDegrees=45
  swing2Motion=Free
  twistLimitDegrees=45
  twistMotion=Free
 }
 {
  bone1Name=dyn_hair_FR_mstr
  bone2Name=head
  swing1LimitDegrees=45
  swing1Motion=Free
  swing2LimitDegrees=45
  swing2Motion=Free
  twistLimitDegrees=45
  twistMotion=Free
 }
 {bone1Name=dyn_hair_FR_01, bone2Name=dyn_hair_FR_mstr, swing1LimitDegrees=15, swing2LimitDegrees=15, twistLimitDegrees=2}
 {bone1Name=dyn_hair_FR_02, bone2Name=dyn_hair_FR_01, swing1LimitDegrees=15, swing2LimitDegrees=15, twistLimitDegrees=2}
 {bone1Name=dyn_hair_FR_03, bone2Name=dyn_hair_FR_02, swing1LimitDegrees=20, swing2LimitDegrees=20, twistLimitDegrees=2}
 {bone1Name=dyn_hair_FR_04, bone2Name=dyn_hair_FR_03, swing1LimitDegrees=30, swing2LimitDegrees=30, twistLimitDegrees=2}
 {bone1Name=dyn_hair_FR_05, bone2Name=dyn_hair_FR_04, swing1LimitDegrees=30, swing2LimitDegrees=30, twistLimitDegrees=2}
 {bone1Name=dyn_hair_FR_06, bone2Name=dyn_hair_FR_05, swing1LimitDegrees=45, swing2LimitDegrees=45, twistLimitDegrees=2}
 {
  bone1Name=upperarm_r
  bone2Name=spine_05
  swing1LimitDegrees=45
  swing1Motion=Free
  swing2LimitDegrees=45
  swing2Motion=Free
  twistLimitDegrees=45
  twistMotion=Free
 }
 {
  bone1Name=upperarm_l
  bone2Name=spine_05
  swing1LimitDegrees=45
  swing1Motion=Free
  swing2LimitDegrees=45
  swing2Motion=Free
  twistLimitDegrees=45
  twistMotion=Free
 }
 {bone1Name=dyn_hair_B_01_02, bone2Name=dyn_hair_B_01_01, swing1LimitDegrees=15, swing2LimitDegrees=20, twistLimitDegrees=2}
 {bone1Name=dyn_hair_B_01_01, bone2Name=dyn_hair_B_01_mstr, swing1LimitDegrees=15, swing2LimitDegrees=15, twistLimitDegrees=2}
 {bone1Name=dyn_hair_B_02_02, bone2Name=dyn_hair_B_02_01, swing1LimitDegrees=15, swing2LimitDegrees=20, twistLimitDegrees=2}
 {bone1Name=dyn_hair_B_03_02, bone2Name=dyn_hair_B_03_01, swing1LimitDegrees=15, swing2LimitDegrees=20, twistLimitDegrees=2}
 {bone1Name=dyn_hair_B_04_02, bone2Name=dyn_hair_B_04_01, swing1LimitDegrees=15, swing2LimitDegrees=20, twistLimitDegrees=2}
 {bone1Name=dyn_hair_B_05_02, bone2Name=dyn_hair_B_05_01, swing1LimitDegrees=15, swing2LimitDegrees=20, twistLimitDegrees=2}
 {bone1Name=dyn_hair_B_06_02, bone2Name=dyn_hair_B_06_01, swing1LimitDegrees=15, swing2LimitDegrees=20, twistLimitDegrees=2}
 {bone1Name=dyn_hair_B_02_01, bone2Name=dyn_hair_B_02_mstr, swing1LimitDegrees=15, swing2LimitDegrees=15, twistLimitDegrees=2}
 {bone1Name=dyn_hair_B_03_01, bone2Name=dyn_hair_B_03_mstr, swing1LimitDegrees=15, swing2LimitDegrees=15, twistLimitDegrees=2}
 {bone1Name=dyn_hair_B_04_01, bone2Name=dyn_hair_B_04_mstr, swing1LimitDegrees=15, swing2LimitDegrees=15, twistLimitDegrees=2}
 {bone1Name=dyn_hair_B_05_01, bone2Name=dyn_hair_B_05_mstr, swing1LimitDegrees=15, swing2LimitDegrees=15, twistLimitDegrees=2}
 {bone1Name=dyn_hair_B_06_01, bone2Name=dyn_hair_B_06_mstr, swing1LimitDegrees=15, swing2LimitDegrees=15, twistLimitDegrees=2}
]
